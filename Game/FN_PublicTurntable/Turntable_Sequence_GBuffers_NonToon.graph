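# Level Sequence for a non-toon turntable that spins the Turntable_UserObjects actor before a fixed-FOV camera, apparently to present GBuffer views.
# The yaw makes one slow turn over 180 frames, then seven faster turns through frame 390.
# Events on BP_Turntable_GBuffers fire at roughly 30-frame steps during the fast turns, one per revolution.
# BP_LevelsChecker events, a master event track and level visibility sections, switching in the final segments, accompany the camera cut.
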LevelSequence /Game/FN_PublicTurntable/Turntable_Sequence_GBuffers_NonToon
bindings [
 CameraActor {id=#1, possessed_class=CameraActor, tracks []}
 BP_LevelsChecker {
  id=#2
  possessed_class=BP_LevelsChecker_C
  tracks [{
    type=MovieSceneEventTrack
    sections [{
      type=MovieSceneEventTriggerSection
      channels [None {
        type=MovieSceneScriptingEventChannel
        keys [
         [0, "<Struct 'MovieSceneEvent' (0x49e5f371ea0) {}>"]
         [180, "<Struct 'MovieSceneEvent' (0x49e5f371680) {}>"]
         [210, "<Struct 'MovieSceneEvent' (0x49e5f371400) {}>"]
        ]
       }]
     }]
   }]
 }
 CameraComponent {
  id=#3
  parent_id=#1
  possessed_class=CameraComponent
  tracks [{
    type=MovieSceneFloatTrack
    display_name=Field Of View
    sections [{type=MovieSceneFloatSection, channels [None {type=MovieSceneScriptingFloatChannel, default=35}]}]
   }]
 }
 Turntable_UserObjects {
  id=#4
  possessed_class=Actor
  tracks [{
    type=MovieScene3DTransformTrack
    sections [{
      type=MovieScene3DTransformSection
      channels [
       Location.X {type=MovieSceneScriptingDoubleChannel, default=0}
       Location.Y {type=MovieSceneScriptingDoubleChannel, default=0}
       Location.Z {type=MovieSceneScriptingDoubleChannel, default=0}
       Rotation.X {type=MovieSceneScriptingDoubleChannel, default=0}
       Rotation.Y {type=MovieSceneScriptingDoubleChannel, default=0}
       Rotation.Z {type=MovieSceneScriptingDoubleChannel, default=2790, thinned_from=9, keys [[0,-90],[180,270],[390,2790]]}
       Scale.X {type=MovieSceneScriptingDoubleChannel, default=1}
       Scale.Y {type=MovieSceneScriptingDoubleChannel, default=1}
       Scale.Z {type=MovieSceneScriptingDoubleChannel, default=1}
      ]
     }]
   }]
 }
 BP_Turntable_GBuffers {
  id=#5
  possessed_class=BP_Turntable_GBuffers_C
  tracks [{
    type=MovieSceneEventTrack
    sections [{
      type=MovieSceneEventTriggerSection
      channels [None {
        type=MovieSceneScriptingEventChannel
        keys [
         [0, "<Struct 'MovieSceneEvent' (0x49e5f371540) {}>"]
         [180, "<Struct 'MovieSceneEvent' (0x49e5f3715e0) {}>"]
         [210, "<Struct 'MovieSceneEvent' (0x49e5f371220) {}>"]
         [240, "<Struct 'MovieSceneEvent' (0x49e5f370000) {}>"]
         [270, "<Struct 'MovieSceneEvent' (0x49e5f371f40) {}>"]
         [300, "<Struct 'MovieSceneEvent' (0x49e5f371fe0) {}>"]
         [331, "<Struct 'MovieSceneEvent' (0x49e5f372080) {}>"]
         [361, "<Struct 'MovieSceneEvent' (0x49e5f372120) {}>"]
         [390, "<Struct 'MovieSceneEvent' (0x49e5f3721c0) {}>"]
        ]
       }]
     }]
   }]
 }
]
display_rate {denominator=1, numerator=30}
tick_resolution {denominator=1, numerator=24000}
tracks [
 {
  type=MovieSceneLevelVisibilityTrack
  sections [
   {type=MovieSceneLevelVisibilitySection, end_frame=360, start_frame=330}
   {type=MovieSceneLevelVisibilitySection, end_frame=330, start_frame=-10}
   {type=MovieSceneLevelVisibilitySection, end_frame=360, start_frame=330}
   {type=MovieSceneLevelVisibilitySection, end_frame=330, start_frame=-10}
   {type=MovieSceneLevelVisibilitySection, end_frame=390, start_frame=360}
   {type=MovieSceneLevelVisibilitySection, end_frame=390, start_frame=360}
   {type=MovieSceneLevelVisibilitySection, end_frame=780, start_frame=390}
   {type=MovieSceneLevelVisibilitySection, end_frame=780, start_frame=390}
  ]
 }
 {
  type=MovieSceneEventTrack
  sections [{
    type=MovieSceneEventTriggerSection
    channels [None {type=MovieSceneScriptingEventChannel, const ["<Struct 'MovieSceneEvent' (0x49e5f370640) {}>", 1]}]
   }]
 }
 {type=MovieSceneCameraCutTrack, sections [{type=MovieSceneCameraCutSection, end_frame=393, start_frame=-10}]}
]
