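# Level Sequence for a 30 fps modelling turntable that shows user objects from the side, top and bottom.
# Turntable_UserObjects spins one full turn for the side view, then two faster turns across the overhead and underside views.
# A fixed-FOV CameraActor jumps between those positions by keys one frame apart under a single camera cut, returning to the side at the end.
# Level visibility sections, a master event track and a BP_Turntable_GBuffers event track accompany it.
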LevelSequence /Game/FN_PublicTurntable/Turntable_Sequence_Modelling
bindings [
 CameraActor {
  id=#1
  possessed_class=CameraActor
  tracks [{
    type=MovieScene3DTransformTrack
    sections [{
      type=MovieScene3DTransformSection
      channels [
       Location.X {type=MovieSceneScriptingDoubleChannel, default=0, thinned_from=7, keys [[0,0.0001],[361,0.0001]]}
       Location.Y {type=MovieSceneScriptingDoubleChannel, default=0, thinned_from=7, keys [[0,539.7187],[240,539.7187],[241,0],[360,0],[361,539.7187]]}
       Location.Z {type=MovieSceneScriptingDoubleChannel, default=-300, keys [[0,175],[240,175],[241,300],[300,300],[301,-300],[360,-300],[361,175]]}
       Rotation.X {type=MovieSceneScriptingDoubleChannel, default=0, thinned_from=7, keys [[0,0],[361,0]]}
       Rotation.Y {type=MovieSceneScriptingDoubleChannel, default=90, keys [[0,-9.5],[240,-9.5],[241,-90],[300,-90],[301,90],[360,90],[361,-9.5]]}
       Rotation.Z {type=MovieSceneScriptingDoubleChannel, default=0, thinned_from=7, keys [[0,-90.2],[240,-90.2],[241,0],[360,0],[361,-90.2]]}
       Scale.X {type=MovieSceneScriptingDoubleChannel, default=1, const [1,7]}
       Scale.Y {type=MovieSceneScriptingDoubleChannel, default=1, const [1,7]}
       Scale.Z {type=MovieSceneScriptingDoubleChannel, default=1, const [1,7]}
      ]
     }]
   }]
 }
 CameraComponent {
  id=#2
  parent_id=#1
  possessed_class=CameraComponent
  tracks [{
    type=MovieSceneFloatTrack
    display_name=Field Of View
    sections [{type=MovieSceneFloatSection, channels [None {type=MovieSceneScriptingFloatChannel, default=35}]}]
   }]
 }
 Turntable_UserObjects {
  id=#3
  possessed_class=Actor
  tracks [{
    type=MovieScene3DTransformTrack
    sections [{
      type=MovieScene3DTransformSection
      channels [
       Location.X {type=MovieSceneScriptingDoubleChannel, default=0}
       Location.Y {type=MovieSceneScriptingDoubleChannel, default=0}
       Location.Z {type=MovieSceneScriptingDoubleChannel, default=0}
       Rotation.X {type=MovieSceneScriptingDoubleChannel, default=0}
       Rotation.Y {type=MovieSceneScriptingDoubleChannel, default=0}
       Rotation.Z {type=MovieSceneScriptingDoubleChannel, default=990, thinned_from=4, keys [[0,-90],[240,270],[360,990]]}
       Scale.X {type=MovieSceneScriptingDoubleChannel, default=1}
       Scale.Y {type=MovieSceneScriptingDoubleChannel, default=1}
       Scale.Z {type=MovieSceneScriptingDoubleChannel, default=1}
      ]
     }]
   }]
 }
 BP_Turntable_GBuffers {
  id=#4
  possessed_class=BP_Turntable_GBuffers_C
  tracks [{type=MovieSceneEventTrack, sections [{type=MovieSceneEventTriggerSection, channels [None {type=MovieSceneScriptingEventChannel}]}]}]
 }
]
display_rate {denominator=1, numerator=30}
tick_resolution {denominator=1, numerator=24000}
tracks [
 {
  type=MovieSceneLevelVisibilityTrack
  sections [
   {type=MovieSceneLevelVisibilitySection, end_frame=393, start_frame=-10}
   {type=MovieSceneLevelVisibilitySection, end_frame=393, start_frame=-10}
  ]
 }
 {
  type=MovieSceneEventTrack
  sections [{
    type=MovieSceneEventTriggerSection
    channels [None {type=MovieSceneScriptingEventChannel, const ["<Struct 'MovieSceneEvent' (0x49e5f372bc0) {}>", 1]}]
   }]
 }
 {type=MovieSceneCameraCutTrack, sections [{type=MovieSceneCameraCutSection, end_frame=361, start_frame=-10}]}
]
